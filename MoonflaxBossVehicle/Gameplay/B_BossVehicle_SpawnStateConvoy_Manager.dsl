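; Blueprint manager that tracks spawned convoy boss vehicles and relays their spawn and garage door events.
; NotifyBossSpawn appends the incoming boss tag container to the spawned-bosses tag set, then broadcasts OnSpawnedBoss with those tags.
; NotifyDoorOpened takes a single boss tag and broadcasts OnGarageDoorOpened, storing nothing.
; BeginPlay only calls the parent implementation.
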(event EventBeginPlay
  (|CallParentFunction))

(event Custom|NotifyBossSpawn
  (params (BossTag :type Gameplay%20Tag%20Container%20Structure))
  (GameplayTags|AppendGameplayTagContainers (Variables|Default|GetSpawnedBossesTags) BossTag)
  (Default|CallOnSpawnedBoss BossTag))

(event Custom|NotifyDoorOpened
  (params (BossTag :type Gameplay%20Tag%20Structure))
  (Default|CallOnGarageDoorOpened BossTag))
